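# Physics asset for the PlatypusBranch female medium back coat, apparently for sliding, simulating coat tails against the lower body.
# Three back chains hang from pelvis-linked base bones and two side chains from thigh-linked master bones, with limited swing and mostly locked twist.
# Free cross-links join neighboring chains near their tips.
# Pelvis and leg capsules provide body shapes, while a huge sunken root sphere approximates a floor.
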PhysicsAsset /BRCosmetics/Characters/Player/Female/Medium/Bodies/F_MED_PlatypusBranch_Soft/Meshes/F_MED_PlatypusBranch_BackCoat_Sliding_Physics
bodies {
 base_coat_bk [{shapeType=Capsule, length=1, radius=1, center=<1 0 0>, rotation=<rot 90 0 0>}]
 base_coat_bk_l [{shapeType=Capsule, length=1, radius=1, center=<1 0 0>, rotation=<rot 90 0 0>}]
 base_coat_bk_r [{shapeType=Capsule, length=1, radius=1, center=<1 0 0>, rotation=<rot 90 0 0>}]
 calf_l [{shapeType=Capsule, length=32.821, radius=11, shapeName=calf_l_capsule, center=<-10 -4 -0.0036>, rotation=<rot 90 0 0>}]
 calf_r [{shapeType=Capsule, length=32.821, radius=11, shapeName=calf_l_capsule, center=<10 4 -0.0036>, rotation=<rot 90 0 0>}]
 dyn_coat_bk_01 [{shapeType=Capsule, length=12, radius=7, center=<10 0 8>, rotation=<rot 0 0 90>}]
 dyn_coat_bk_02 [{shapeType=Capsule, length=12, radius=7, center=<10 0 8>, rotation=<rot 0 0 90>}]
 dyn_coat_bk_03 [{shapeType=Capsule, length=12, radius=7, center=<10 0 5>, rotation=<rot 0 0 90>}]
 dyn_coat_bk_04 [{shapeType=Capsule, length=12, radius=7, center=<10 0 5>, rotation=<rot 0 0 90>}]
 dyn_coat_bk_05 [{shapeType=Capsule, length=12, radius=7, center=<10 0 5>, rotation=<rot 0 0 90>}]
 dyn_coat_bk_l_01 [{shapeType=Capsule, length=12, radius=7, center=<10 0 12>, rotation=<rot 0 0 90>}]
 dyn_coat_bk_l_02 [{shapeType=Capsule, length=12, radius=7, center=<10 0 12>, rotation=<rot 0 0 90>}]
 dyn_coat_bk_l_03 [{shapeType=Capsule, length=12, radius=7, center=<10 0 10>, rotation=<rot 0 0 90>}]
 dyn_coat_bk_l_04 [{shapeType=Capsule, length=12, radius=7, center=<10 0 10>, rotation=<rot 0 0 90>}]
 dyn_coat_bk_l_05 [{shapeType=Capsule, length=12, radius=7, center=<10 0 11>, rotation=<rot 0 0 90>}]
 dyn_coat_bk_r_01 [{shapeType=Capsule, length=12, radius=7, center=<10 0 12>, rotation=<rot 0 0 90>}]
 dyn_coat_bk_r_02 [{shapeType=Capsule, length=12, radius=7, center=<10 0 12>, rotation=<rot 0 0 90>}]
 dyn_coat_bk_r_03 [{shapeType=Capsule, length=12, radius=7, center=<10 0 10>, rotation=<rot 0 0 90>}]
 dyn_coat_bk_r_04 [{shapeType=Capsule, length=12, radius=7, center=<10 0 10>, rotation=<rot 0 0 90>}]
 dyn_coat_bk_r_05 [{shapeType=Capsule, length=12, radius=7, center=<10 0 11>, rotation=<rot 0 0 90>}]
 dyn_coat_l_01 [{shapeType=Capsule, length=12, radius=7, center=<10 0 12>, rotation=<rot 0 0 90>}]
 dyn_coat_l_02 [{shapeType=Capsule, length=12, radius=7, center=<10 0 12>, rotation=<rot 0 0 90>}]
 dyn_coat_l_03 [{shapeType=Capsule, length=12, radius=7, center=<10 0 12>, rotation=<rot 0 0 90>}]
 dyn_coat_l_04 [{shapeType=Capsule, length=12, radius=7, center=<10 0 11>, rotation=<rot 0 0 90>}]
 dyn_coat_l_MSTR [{shapeType=Capsule, length=1, radius=1, center=<1 0 0>, rotation=<rot 90 0 0>}]
 dyn_coat_r_01 [{shapeType=Capsule, length=12, radius=7, center=<10 0 12>, rotation=<rot 0 0 90>}]
 dyn_coat_r_02 [{shapeType=Capsule, length=12, radius=7, center=<10 0 12>, rotation=<rot 0 0 90>}]
 dyn_coat_r_03 [{shapeType=Capsule, length=12, radius=7, center=<10 0 12>, rotation=<rot 0 0 90>}]
 dyn_coat_r_04 [{shapeType=Capsule, length=12, radius=7, center=<10 0 11>, rotation=<rot 0 0 90>}]
 dyn_coat_r_MSTR [{shapeType=Capsule, length=1, radius=1, center=<1 0 0>, rotation=<rot 90 0 0>}]
 foot_l [{shapeType=Capsule, length=25, radius=10, shapeName=foot_r_capsule, center=<1.5635 7.5 -0.6813>, rotation=<rot 0 45 90>}]
 foot_r [{shapeType=Capsule, length=25, radius=12, shapeName=foot_r_capsule, center=<1.5635 -3.0425 -0.6813>, rotation=<rot 0 45 90>}]
 pelvis [{shapeType=Capsule, length=9.3507, radius=9.3507, shapeName=pelvis_capsule, center=<-0.0465 -1.2985 0>}]
 root [{shapeType=Sphere, radius=500, shapeName=root_sphere, center=<0 0 -491.9244>}]
 thigh_l [{shapeType=Capsule, length=25.7046, radius=9.2388, shapeName=thigh_l_capsule, center=<-15 -0.6013 1.7204>, rotation=<rot -90 0 0>}]
 thigh_r [{shapeType=Capsule, length=25.7046, radius=9.2346, shapeName=thigh_r_capsule, center=<15 0.6014 -1.7811>, rotation=<rot -90 0 0>}]
}
constraints [
 {
  bone1Name=base_coat_bk
  bone2Name=pelvis
  swing1LimitDegrees=45
  swing1Motion=Free
  swing2LimitDegrees=45
  swing2Motion=Free
  twistLimitDegrees=45
  twistMotion=Free
 }
 {bone1Name=dyn_coat_bk_01, bone2Name=base_coat_bk, swing1LimitDegrees=10, swing2LimitDegrees=20, twistLimitDegrees=2}
 {bone1Name=dyn_coat_bk_02, bone2Name=dyn_coat_bk_01, swing1LimitDegrees=40, swing2LimitDegrees=20, twistLimitDegrees=2, twistMotion=Locked}
 {bone1Name=dyn_coat_bk_03, bone2Name=dyn_coat_bk_02, swing1LimitDegrees=40, swing2LimitDegrees=20, twistLimitDegrees=2, twistMotion=Locked}
 {bone1Name=dyn_coat_bk_04, bone2Name=dyn_coat_bk_03, swing1LimitDegrees=40, swing2LimitDegrees=30, twistLimitDegrees=2, twistMotion=Locked}
 {bone1Name=dyn_coat_bk_05, bone2Name=dyn_coat_bk_04, swing1LimitDegrees=40, swing2LimitDegrees=30, twistLimitDegrees=2, twistMotion=Locked}
 {
  bone1Name=base_coat_bk_l
  bone2Name=pelvis
  swing1LimitDegrees=45
  swing1Motion=Free
  swing2LimitDegrees=45
  swing2Motion=Free
  twistLimitDegrees=45
  twistMotion=Free
 }
 {bone1Name=dyn_coat_bk_l_01, bone2Name=base_coat_bk_l, swing1LimitDegrees=10, swing2LimitDegrees=20, twistLimitDegrees=2}
 {
  bone1Name=dyn_coat_bk_l_02
  bone2Name=dyn_coat_bk_l_01
  swing1LimitDegrees=40
  swing2LimitDegrees=20
  twistLimitDegrees=2
  twistMotion=Locked
 }
 {
  bone1Name=dyn_coat_bk_l_03
  bone2Name=dyn_coat_bk_l_02
  swing1LimitDegrees=40
  swing2LimitDegrees=20
  twistLimitDegrees=2
  twistMotion=Locked
 }
 {
  bone1Name=dyn_coat_bk_l_04
  bone2Name=dyn_coat_bk_l_03
  swing1LimitDegrees=40
  swing2LimitDegrees=30
  twistLimitDegrees=2
  twistMotion=Locked
 }
 {
  bone1Name=dyn_coat_bk_l_05
  bone2Name=dyn_coat_bk_l_04
  swing1LimitDegrees=40
  swing2LimitDegrees=30
  twistLimitDegrees=2
  twistMotion=Locked
 }
 {
  bone1Name=base_coat_bk_r
  bone2Name=pelvis
  swing1LimitDegrees=45
  swing1Motion=Free
  swing2LimitDegrees=45
  swing2Motion=Free
  twistLimitDegrees=45
  twistMotion=Free
 }
 {bone1Name=dyn_coat_bk_r_01, bone2Name=base_coat_bk_r, swing1LimitDegrees=10, swing2LimitDegrees=20, twistLimitDegrees=2}
 {
  bone1Name=dyn_coat_bk_r_02
  bone2Name=dyn_coat_bk_r_01
  swing1LimitDegrees=40
  swing2LimitDegrees=20
  twistLimitDegrees=2
  twistMotion=Locked
 }
 {
  bone1Name=dyn_coat_bk_r_03
  bone2Name=dyn_coat_bk_r_02
  swing1LimitDegrees=40
  swing2LimitDegrees=20
  twistLimitDegrees=2
  twistMotion=Locked
 }
 {
  bone1Name=dyn_coat_bk_r_04
  bone2Name=dyn_coat_bk_r_03
  swing1LimitDegrees=40
  swing2LimitDegrees=30
  twistLimitDegrees=2
  twistMotion=Locked
 }
 {
  bone1Name=dyn_coat_bk_r_05
  bone2Name=dyn_coat_bk_r_04
  swing1LimitDegrees=40
  swing2LimitDegrees=30
  twistLimitDegrees=2
  twistMotion=Locked
 }
 {
  bone1Name=dyn_coat_r_MSTR
  bone2Name=thigh_r
  swing1LimitDegrees=45
  swing1Motion=Free
  swing2LimitDegrees=45
  swing2Motion=Free
  twistLimitDegrees=45
  twistMotion=Free
 }
 {bone1Name=dyn_coat_r_01, bone2Name=dyn_coat_r_MSTR, swing1LimitDegrees=30, swing2LimitDegrees=30, twistLimitDegrees=5, twistMotion=Locked}
 {bone1Name=dyn_coat_r_02, bone2Name=dyn_coat_r_01, swing1LimitDegrees=30, swing2LimitDegrees=30, twistLimitDegrees=2, twistMotion=Locked}
 {bone1Name=dyn_coat_r_03, bone2Name=dyn_coat_r_02, swing1LimitDegrees=30, swing2LimitDegrees=30, twistLimitDegrees=2, twistMotion=Locked}
 {bone1Name=dyn_coat_r_04, bone2Name=dyn_coat_r_03, swing1LimitDegrees=30, swing2LimitDegrees=30, twistLimitDegrees=2, twistMotion=Locked}
 {bone1Name=dyn_coat_l_04, bone2Name=dyn_coat_l_03, swing1LimitDegrees=30, swing2LimitDegrees=30, twistLimitDegrees=2, twistMotion=Locked}
 {
  bone1Name=dyn_coat_l_MSTR
  bone2Name=thigh_l
  swing1LimitDegrees=45
  swing1Motion=Free
  swing2LimitDegrees=45
  swing2Motion=Free
  twistLimitDegrees=45
  twistMotion=Free
 }
 {bone1Name=dyn_coat_l_01, bone2Name=dyn_coat_l_MSTR, swing1LimitDegrees=30, swing2LimitDegrees=30, twistLimitDegrees=5, twistMotion=Locked}
 {bone1Name=dyn_coat_l_02, bone2Name=dyn_coat_l_01, swing1LimitDegrees=30, swing2LimitDegrees=30, twistLimitDegrees=2, twistMotion=Locked}
 {bone1Name=dyn_coat_l_03, bone2Name=dyn_coat_l_02, swing1LimitDegrees=30, swing2LimitDegrees=30, twistLimitDegrees=2, twistMotion=Locked}
 {
  bone1Name=dyn_coat_bk_l_04
  bone2Name=dyn_coat_bk_04
  swing1LimitDegrees=45
  swing1Motion=Free
  swing2LimitDegrees=45
  swing2Motion=Free
  twistLimitDegrees=45
  twistMotion=Free
 }
 {
  bone1Name=dyn_coat_bk_r_04
  bone2Name=dyn_coat_bk_04
  swing1LimitDegrees=45
  swing1Motion=Free
  swing2LimitDegrees=45
  swing2Motion=Free
  twistLimitDegrees=45
  twistMotion=Free
 }
 {
  bone1Name=dyn_coat_r_04
  bone2Name=dyn_coat_bk_r_05
  swing1LimitDegrees=45
  swing1Motion=Free
  swing2LimitDegrees=45
  swing2Motion=Free
  twistLimitDegrees=45
  twistMotion=Free
 }
 {
  bone1Name=dyn_coat_l_04
  bone2Name=dyn_coat_bk_l_05
  swing1LimitDegrees=45
  swing1Motion=Free
  swing2LimitDegrees=45
  swing2Motion=Free
  twistLimitDegrees=45
  twistMotion=Free
 }
 {
  bone1Name=pelvis
  bone2Name=root
  swing1LimitDegrees=45
  swing1Motion=Free
  swing2LimitDegrees=45
  swing2Motion=Free
  twistLimitDegrees=45
  twistMotion=Free
 }
 {
  bone1Name=thigh_l
  bone2Name=root
  swing1LimitDegrees=45
  swing1Motion=Free
  swing2LimitDegrees=45
  swing2Motion=Free
  twistLimitDegrees=45
  twistMotion=Free
 }
 {
  bone1Name=calf_l
  bone2Name=root
  swing1LimitDegrees=45
  swing1Motion=Free
  swing2LimitDegrees=45
  swing2Motion=Free
  twistLimitDegrees=45
  twistMotion=Free
 }
 {
  bone1Name=foot_l
  bone2Name=root
  swing1LimitDegrees=45
  swing1Motion=Free
  swing2LimitDegrees=45
  swing2Motion=Free
  twistLimitDegrees=45
  twistMotion=Free
 }
 {
  bone1Name=thigh_r
  bone2Name=root
  swing1LimitDegrees=45
  swing1Motion=Free
  swing2LimitDegrees=45
  swing2Motion=Free
  twistLimitDegrees=45
  twistMotion=Free
 }
 {
  bone1Name=calf_r
  bone2Name=root
  swing1LimitDegrees=45
  swing1Motion=Free
  swing2LimitDegrees=45
  swing2Motion=Free
  twistLimitDegrees=45
  twistMotion=Free
 }
 {
  bone1Name=foot_r
  bone2Name=root
  swing1LimitDegrees=45
  swing1Motion=Free
  swing2LimitDegrees=45
  swing2Motion=Free
  twistLimitDegrees=45
  twistMotion=Free
 }
]
