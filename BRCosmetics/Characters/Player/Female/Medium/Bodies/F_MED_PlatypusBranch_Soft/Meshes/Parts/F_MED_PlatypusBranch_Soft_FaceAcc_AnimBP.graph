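# Anim Blueprint for the F_MED_PlatypusBranch_Soft face accessory, adding secondary motion to hair and hood on a pose copied from the attached parent mesh.
# Head-versus-spine_05 rotation, measured each update, drives additive offsets on base_hair_back to follow head pitch and turn.
# Chained spring-based Anim Dynamics nodes then simulate hair and hood bones before a hood scale tweak.
# Zipline enter and exit input events switch the dynamics alpha off and on.
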AnimBlueprint /BRCosmetics/Characters/Player/Female/Medium/Bodies/F_MED_PlatypusBranch_Soft/Meshes/Parts/F_MED_PlatypusBranch_Soft_FaceAcc_AnimBP
target_skeleton=&/F_MED_PlatypusBranch_Soft_FaceAcc_Skeleton
graphs [
 AnimGraph {nodes [
   n1 = "Transform (Modify) Bone - Bone: base_hair_back"(
    ComponentPose=Local To Component(LocalPose=Copy Pose From Mesh() {bUseAttachedParent=true})
    Alpha=Get headUp() -> (headUp:double)
    Translation="-6,0,0"
    Rotation="25,0,0"
    Scale="1,1,1"
   ) {
    rotationMode=BMM_Additive
    rotationSpace=BCS_BoneSpace
    translationMode=BMM_Additive
    translationSpace=BCS_BoneSpace
    alphaBoolBlend {blendOption=Linear}
    boneToModify {boneName=base_hair_back}
    rotation {pitch=25}
    translation {x=-6}
   }
   n2 = "Transform (Modify) Bone - Bone: base_hair_back"(ComponentPose=@n1, Alpha=Get headLR() -> (headLR:double), Translation="0,0,0", Rotation="2.5,0,0", Scale="1,1,1") {
    rotationMode=BMM_Additive
    rotationSpace=BCS_BoneSpace
    translationMode=BMM_Additive
    translationSpace=BCS_BoneSpace
    alphaBoolBlend {blendOption=Linear}
    boneToModify {boneName=base_hair_back}
    rotation {pitch=2.5}
   }
   n3 = Get animD alpha() -> (animD alpha:double)
   n4 = Anim Dynamics - Bone: dyn_hair_r(ComponentPose=@n2, Alpha=@n3, ExternalForce="0,0,0") {
    angularDampingOverride=1
    angularSpringConstant=10000
    bAngularSpring=true
    bOverrideAngularDamping=true
    bUseGravityOverride=true
    gravityScale=0.35
    simulationSpace=RootRelative
    alphaBoolBlend {blendOption=Linear}
    boundBone {boneName=dyn_hair_r}
    componentLinearAccScale=<1 1 1>
    componentLinearVelScale=<1 1 1>
    gravityOverride {z=-980}
    physicsBodyDefinitions [{
      boundBone {boneName=dyn_hair_r}
      boxExtents=<25 25 25>
      constraintSetup {angularLimitsMax=<0 30 30>, angularLimitsMin=<-30 -30 -30>, angularTarget=<1 0 0>}
      localJointOffset=<3 0 0>
     }]
   }
   n5 = Anim Dynamics - Bone: dyn_hair_frnt_r(ComponentPose=@n4, Alpha=@n3, ExternalForce="0,0,0") {
    angularDampingOverride=1
    angularSpringConstant=10000
    bAngularSpring=true
    bOverrideAngularDamping=true
    bUseGravityOverride=true
    gravityScale=0.35
    simulationSpace=RootRelative
    alphaBoolBlend {blendOption=Linear}
    boundBone {boneName=dyn_hair_frnt_r}
    componentLinearAccScale=<1 1 1>
    componentLinearVelScale=<1 1 1>
    gravityOverride {z=-980}
    physicsBodyDefinitions [{
      boundBone {boneName=dyn_hair_frnt_r}
      boxExtents=<25 25 25>
      constraintSetup {angularLimitsMax=<30 30 30>, angularLimitsMin=<-30 -30 -30>, angularTarget=<1 0 0>}
      localJointOffset=<3 0 0>
     }]
   }
   n6 = Anim Dynamics - Bone: dyn_hair_frnt_l(ComponentPose=@n5, Alpha=@n3, ExternalForce="0,0,0") {
    angularDampingOverride=1
    angularSpringConstant=10000
    bAngularSpring=true
    bOverrideAngularDamping=true
    bUseGravityOverride=true
    gravityScale=0.35
    simulationSpace=RootRelative
    alphaBoolBlend {blendOption=Linear}
    boundBone {boneName=dyn_hair_frnt_l}
    componentLinearAccScale=<1 1 1>
    componentLinearVelScale=<1 1 1>
    gravityOverride {z=-980}
    physicsBodyDefinitions [{
      boundBone {boneName=dyn_hair_frnt_l}
      boxExtents=<25 25 25>
      constraintSetup {angularLimitsMax=<0 0 30>, angularLimitsMin=<-30 -30 -30>, angularTarget=<1 0 0>}
      localJointOffset=<3 0 0>
     }]
   }
   n7 = Anim Dynamics - Bone: dyn_hair_back(ComponentPose=@n6, Alpha=@n3, ExternalForce="0,0,0") {
    angularDampingOverride=1
    angularSpringConstant=10000
    bAngularSpring=true
    bOverrideAngularDamping=true
    bUseGravityOverride=true
    gravityScale=0.35
    simulationSpace=RootRelative
    alphaBoolBlend {blendOption=Linear}
    boundBone {boneName=dyn_hair_back}
    componentLinearAccScale=<1 1 1>
    componentLinearVelScale=<1 1 1>
    gravityOverride {z=-980}
    physicsBodyDefinitions [{
      boundBone {boneName=dyn_hair_back}
      boxExtents=<25 25 25>
      constraintSetup {angularLimitsMax=<30 15 30>, angularLimitsMin=<-30 -15 -30>, angularTarget=<1 0 0>}
      localJointOffset=<3 0 0>
     }]
   }
   n8 = Anim Dynamics - Bone: dyn_hair_l(ComponentPose=@n7, Alpha=@n3, ExternalForce="0,0,0") {
    angularDampingOverride=1
    angularSpringConstant=10000
    bAngularSpring=true
    bOverrideAngularDamping=true
    bUseGravityOverride=true
    gravityScale=0.35
    simulationSpace=RootRelative
    alphaBoolBlend {blendOption=Linear}
    boundBone {boneName=dyn_hair_l}
    componentLinearAccScale=<1 1 1>
    componentLinearVelScale=<1 1 1>
    gravityOverride {z=-980}
    physicsBodyDefinitions [{
      boundBone {boneName=dyn_hair_l}
      boxExtents=<25 25 25>
      constraintSetup {angularLimitsMax=<0 0 30>, angularLimitsMin=<-30 -30 -30>, angularTarget=<1 0 0>}
      localJointOffset=<3 0 0>
     }]
   }
   n9 = Anim Dynamics - Bone: dyn_hair_top_r(ComponentPose=@n8, Alpha=@n3, ExternalForce="0,0,0") {
    angularDampingOverride=1
    angularSpringConstant=10000
    bAngularSpring=true
    bOverrideAngularDamping=true
    bUseGravityOverride=true
    gravityScale=0.35
    simulationSpace=RootRelative
    alphaBoolBlend {blendOption=Linear}
    boundBone {boneName=dyn_hair_top_r}
    componentLinearAccScale=<1 1 1>
    componentLinearVelScale=<1 1 1>
    gravityOverride {z=-980}
    physicsBodyDefinitions [{
      boundBone {boneName=dyn_hair_top_r}
      boxExtents=<60 60 60>
      constraintSetup {angularLimitsMax=<0 0 10>, angularLimitsMin=<-10 -10 -10>, angularTarget=<1 0 0>}
      localJointOffset=<3 0 0>
     }]
   }
   n10 = Anim Dynamics - Bone: dyn_hair_top(ComponentPose=@n9, Alpha=@n3, ExternalForce="0,0,0") {
    angularDampingOverride=1
    angularSpringConstant=10000
    bAngularSpring=true
    bOverrideAngularDamping=true
    bUseGravityOverride=true
    gravityScale=0.35
    simulationSpace=RootRelative
    alphaBoolBlend {blendOption=Linear}
    boundBone {boneName=dyn_hair_top}
    componentLinearAccScale=<1 1 1>
    componentLinearVelScale=<1 1 1>
    gravityOverride {z=-980}
    physicsBodyDefinitions [{
      boundBone {boneName=dyn_hair_top}
      boxExtents=<60 60 60>
      constraintSetup {angularLimitsMax=<10 0 10>, angularLimitsMin=<-10 -10 -10>, angularTarget=<1 0 0>}
      localJointOffset=<3 0 0>
     }]
   }
   n11 = Anim Dynamics - Bone: dyn_hood_on_bk(ComponentPose=@n10, Alpha=@n3, ExternalForce="0,0,0") {
    angularDampingOverride=1
    angularSpringConstant=3500
    bAngularSpring=true
    bOverrideAngularDamping=true
    bUseGravityOverride=true
    gravityScale=0.35
    simulationSpace=RootRelative
    alphaBoolBlend {blendOption=Linear}
    boundBone {boneName=dyn_hood_on_bk}
    componentAppliedLinearAccClamp=<2500 2500 1500>
    componentLinearAccScale=<1 1 1>
    componentLinearVelScale=<1 1 1>
    gravityOverride {z=-980}
    physicsBodyDefinitions [{
      boundBone {boneName=dyn_hood_on_bk}
      boxExtents=<25 25 25>
      constraintSetup {angularLimitsMax=<2.5 25 2.5>, angularLimitsMin=<-2.5 -25 -2.5>, angularTarget=<1 0 0>}
      localJointOffset=<4 0 0>
     }]
   }
   n12 = "Transform (Modify) Bone - Bone: base_hood_on_bk"!(ComponentPose=@n11, bAlphaBoolEnabled:bool=True, Translation="0,0,0", Rotation="0,0,0", Scale="1,1.2,1") -> (Pose) {
    alphaInputType=Bool
    rotationMode=BMM_Additive
    rotationSpace=BCS_BoneSpace
    scaleMode=BMM_Additive
    scaleSpace=BCS_BoneSpace
    translationMode=BMM_Additive
    translationSpace=BCS_BoneSpace
    alphaBoolBlend {blendInTime=0.2, blendOption=Linear, blendOutTime=0.2}
    boneToModify {boneName=base_hood_on_bk}
    scale {y=1.2}
   }
   Output Pose(Result=Component To Local(ComponentPose=@n12)) {layerGroup=None}
  ]}
 EventGraph {nodes [
   Event BlueprintUpdateAnimation()
   > n1 = Sequence()
   > n2 = Branch(Condition=Get OnZipline() -> (OnZipline:bool))
   > Set animD alpha(animD alpha:double=0) -> (Output_Get:double)
   Set animD alpha(execute=@n2.else, animD alpha=1)
   n3 = Get Skel Mesh()
   Set headUp(
    execute=@n1.then_1
    headUp:double=GetAlphaFromReferenceTransformDelta(SkeletalMeshComponent=@n3, BoneName=spine_05, BaseName=head, InputRange=(X=0,Y=90), TwistAxis="0,0,1")
   ) -> (Output_Get:double)
   > Set headLR(headLR:double="Absolute (Float)"(A=GetAlphaFromReferenceTransformDelta(SkeletalMeshComponent=@n3, BoneName=spine_05, BaseName=head, DeltaComponentType=EulerY, InputRange=(X=0,Y=90), TwistAxis="0,0,1").RawOutput)) -> (Output_Get:double)
   Event BlueprintInitializeAnimation()
   > Set Skel Mesh(Skel Mesh=GetOwnerSkeletalMeshComponent())
   n4 = Event BlueprintProcessAnimInputEvent()
   > n5 = Branch(Condition=IsNamed(self=@n4.AnimInputEvent, InEventName=aie.Player.Zipline.Entered))
   > Set OnZipline(OnZipline:bool=1) -> (Output_Get:bool)
   Branch(execute=@n5.else, Condition=IsNamed(self=@n4.AnimInputEvent, InEventName=aie.Player.Zipline.Exited))
   > Set OnZipline(OnZipline=false)
  ]}
]
variables [
 animD alpha {type=number}
 headLR {type=number}
 headUp {type=number}
 OnZipline {type=boolean}
 Skel Mesh {type=SkeletalMeshComponent}
]
